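# Material function for electric gadget props that applies a parameterized restyle to selected areas of an incoming material.
# It breaks out the input attributes, then adjusts base color (invert, desaturate, brightness, offset, or flat override) and remaps metallic, specular and roughness between min/max parameters.
# A channel-mask parameter, Ranked Mask, selects a channel of the Mask input to blend the adjusted attributes over the original material.
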MaterialFunction /BRCosmetics/Animation/Props/Gadgets_Electric/Materials/MF_Gadget_Electric
graph {nodes [
  n1 = FunctionInput() {inputName=Material, inputType=FunctionInput_MaterialAttributes}
  n2 = GetMaterialAttributes(None=@n1) {attributeGetTypes [
    #1
    #2
    #3
    #4
   ]}
  n3 = VectorParameter() {group=Gadget Electric, parameterName=Base Color Override, defaultValue=<rgba 0 0 0 1>}
  n4 = StaticSwitchParameter(True=OneMinus(None=@n2.BaseColor), False=@n2.BaseColor) {group=Gadget Electric, parameterName=Invert Base Color}
  n5 = ScalarParameter() {group=Gadget Electric, parameterName=Base Color Desaturation}
  n6 = Add(
   A=Multiply(A=Desaturation(None=@n4, Fraction=@n5), B=ScalarParameter() {defaultValue=1, group=Gadget Electric, parameterName=Base Color Brightness})
   B=ScalarParameter() {group=Gadget Electric, parameterName=Base Color Offset}
  )
  n7 = Multiply(
   A=LinearInterpolate(
    A=ScalarParameter() {group=Gadget Electric, parameterName=Roughness Min}
    B=ScalarParameter() {defaultValue=1, group=Gadget Electric, parameterName=Roughness Max}
    Alpha=@n2.Roughness
   )
   B=ScalarParameter() {defaultValue=1, group=Gadget Electric, parameterName=Roughness Bias}
  )
  n8 = ScalarParameter() {group=Gadget Electric, parameterName=Roughness Offset}
  n9 = SetMaterialAttributes!(
   MaterialAttributes=@n2.MaterialAttributes
   BaseColor=Saturate(None=StaticSwitchParameter(True=@n3.RGB, False=@n6) {group=Gadget Electric, parameterName=Override Base Color})
   Metallic=LinearInterpolate(
    A=ScalarParameter() {group=Gadget Electric, parameterName=Metallic Min}
    B=ScalarParameter() {defaultValue=1, group=Gadget Electric, parameterName=Metallic Max}
    Alpha=@n2.Metallic
   )
   Specular=LinearInterpolate(
    A=ScalarParameter() {group=Gadget Electric, parameterName=Specular Min}
    B=ScalarParameter() {defaultValue=1, group=Gadget Electric, parameterName=Specular Max}
    Alpha=@n2.Specular
   )
   Roughness=Saturate(None=Add(A=@n7, B=@n8))
  ) {attributeSetTypes [
    #1
    #2
    #3
    #4
   ]}
  n10 = FunctionInput() {inputName=Mask, sortPriority=1}
  FunctionOutput(None=BlendMaterialAttributes(A=@n1, B=@n9, Alpha=ChannelMaskParameter(None=@n10) {group=Gadget Electric, parameterName=Ranked Mask, sortPriority=0}.R))
 ]}
